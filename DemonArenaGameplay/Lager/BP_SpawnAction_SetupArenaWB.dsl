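; A Living World spawn action that sets up a Fire Petal boss for the demon arena when it spawns.
; On spawn it casts the actor to the non-participant boss pawn and continues only if it carries the FirePetal WB boss type tag.
; Matching bosses get a loose "IsInArena" gameplay tag, have their health display hidden, and are moved from all factions into the Fire Petal hostile faction.
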(fn OnActorSpawned
  (params (SpawnedActor :type Actor%20Object%20Reference) (SourcePointProvider :type Fort%20Athena%20Living%20World%20Point%20Provider%20Interface%20Interface%20%28by%20ref%29))
  (bind _asbp_player_pawn_non_participant_boss (Utilities|Casting|CastToBP_PlayerPawn_NonParticipant_Boss SpawnedActor)
    (:then
      (if (GameplayTags|HasMatchingGameplayTag _asbp_player_pawn_non_participant_boss "(TagName=\"AISpawnerData.Type.Boss.FirePetal.WB\")")
        (Ability|GameplayEffect|AddLooseGameplayTags _asbp_player_pawn_non_participant_boss (GameplayTags|MakeGameplayTagContainerfromTag "(TagName=\"FirePetalBosses.Demons.IsInArena\")") true)
        (Variables|/Game/Athena/AI/NPCs/Base/BP_PlayerPawn_NonParticipant_Boss.BP_PlayerPawn_NonParticipant_Boss_C|SetShowHealth false _asbp_player_pawn_non_participant_boss)
        (Affiliation|RemovefromAllFactions _asbp_player_pawn_non_participant_boss)
        (Affiliation|AddtoFaction _asbp_player_pawn_non_participant_boss "(TagName=\"Athena.Faction.FirePetal.Hostile\")")
        (else
          _))
      (return))
    (:CastFailed
      _)))
